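# Physics asset for the Grandeur Male two-handed pickaxe, giving capsule bodies to the handle and its ten dyn_umbrella bones.
# Six umbrella bodies (1-3, 8-10) are constrained directly to the handle with swing and twist locked, so they stay rigid.
# The other four form two two-link chains off the handle (4 to 5, 6 to 7) with tight angular limits, allowing slight secondary motion.
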PhysicsAsset /BRCosmetics/Weapons/FORT_Melee/Pickaxe_Grandeur_Male/Meshes/Grandeur_Male_Axe_Physics
bodies {
 Melee_TwoHanded_Handle [{shapeType=Capsule, length=20, radius=3.6867, shapeName=Melee_TwoHanded_Handle_capsule}]
 dyn_umbrella1 [{shapeType=Capsule, length=6.6437, radius=3.514, shapeName=dyn_umbrella1_capsule}]
 dyn_umbrella10 [{shapeType=Capsule, length=4.8025, radius=2.8971, shapeName=dyn_umbrella2_capsule}]
 dyn_umbrella2 [{shapeType=Capsule, length=4.8025, radius=2.8971, shapeName=dyn_umbrella2_capsule}]
 dyn_umbrella3 [{shapeType=Capsule, length=4.8025, radius=2.8971, shapeName=dyn_umbrella2_capsule}]
 dyn_umbrella4 [{shapeType=Capsule, length=7.7497, radius=1.82, shapeName=dyn_umbrella6_capsule, rotation=<rot 90 0 0>}]
 dyn_umbrella5 [{shapeType=Capsule, length=7.7497, radius=1.82, shapeName=dyn_umbrella6_capsule, center=<0 0 0.0002>, rotation=<rot 90 0 0>}]
 dyn_umbrella6 [{shapeType=Capsule, length=7.7497, radius=1.82, shapeName=dyn_umbrella6_capsule, rotation=<rot 90 0 0>}]
 dyn_umbrella7 [{shapeType=Capsule, length=7.7497, radius=1.82, shapeName=dyn_umbrella6_capsule, center=<0 0 0.0002>, rotation=<rot 90 0 0>}]
 dyn_umbrella8 [{shapeType=Capsule, length=6.6437, radius=3.514, shapeName=dyn_umbrella1_capsule}]
 dyn_umbrella9 [{shapeType=Capsule, length=4.8025, radius=2.8971, shapeName=dyn_umbrella2_capsule}]
}
constraints [
 {
  bone1Name=dyn_umbrella1
  bone2Name=Melee_TwoHanded_Handle
  swing1LimitDegrees=45
  swing1Motion=Locked
  swing2LimitDegrees=45
  swing2Motion=Locked
  twistLimitDegrees=45
  twistMotion=Locked
 }
 {
  bone1Name=dyn_umbrella2
  bone2Name=Melee_TwoHanded_Handle
  swing1LimitDegrees=45
  swing1Motion=Locked
  swing2LimitDegrees=45
  swing2Motion=Locked
  twistLimitDegrees=45
  twistMotion=Locked
 }
 {
  bone1Name=dyn_umbrella3
  bone2Name=Melee_TwoHanded_Handle
  swing1LimitDegrees=45
  swing1Motion=Locked
  swing2LimitDegrees=45
  swing2Motion=Locked
  twistLimitDegrees=45
  twistMotion=Locked
 }
 {
  bone1Name=dyn_umbrella9
  bone2Name=Melee_TwoHanded_Handle
  swing1LimitDegrees=45
  swing1Motion=Locked
  swing2LimitDegrees=45
  swing2Motion=Locked
  twistLimitDegrees=45
  twistMotion=Locked
 }
 {
  bone1Name=dyn_umbrella10
  bone2Name=Melee_TwoHanded_Handle
  swing1LimitDegrees=45
  swing1Motion=Locked
  swing2LimitDegrees=45
  swing2Motion=Locked
  twistLimitDegrees=45
  twistMotion=Locked
 }
 {
  bone1Name=dyn_umbrella8
  bone2Name=Melee_TwoHanded_Handle
  swing1LimitDegrees=45
  swing1Motion=Locked
  swing2LimitDegrees=45
  swing2Motion=Locked
  twistLimitDegrees=45
  twistMotion=Locked
 }
 {bone1Name=dyn_umbrella6, bone2Name=Melee_TwoHanded_Handle, swing1LimitDegrees=5, swing2LimitDegrees=5, twistLimitDegrees=1}
 {bone1Name=dyn_umbrella7, bone2Name=dyn_umbrella6, swing1LimitDegrees=10, swing2LimitDegrees=5, twistLimitDegrees=1}
 {bone1Name=dyn_umbrella4, bone2Name=Melee_TwoHanded_Handle, swing1LimitDegrees=5, swing2LimitDegrees=5, twistLimitDegrees=1}
 {bone1Name=dyn_umbrella5, bone2Name=dyn_umbrella4, swing1LimitDegrees=10, swing2LimitDegrees=5, twistLimitDegrees=1}
]
